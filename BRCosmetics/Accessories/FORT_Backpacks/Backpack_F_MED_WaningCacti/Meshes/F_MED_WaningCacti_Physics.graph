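# Physics asset for the WaningCacti backpack cosmetic, defining collision and joints for a hanging charm chain on the pack.
# A large capsule covers the pack bone, while small capsules cover charm_main and charm_01 through charm_05.
# Constraints link the bones in sequence from pack through charm_main to the charm_05 tip with limited swing.
# Twist is locked at the pack attachment and the last two joints, and tightly limited along the middle links.
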PhysicsAsset /BRCosmetics/Accessories/FORT_Backpacks/Backpack_F_MED_WaningCacti/Meshes/F_MED_WaningCacti_Physics
bodies {
 charm_01 [{
   shapeType=Capsule
   length=2.284
   radius=1.0163
   shapeName=charm_01_capsule
   center=<1.4267 0.2279 -0.0586>
   rotation=<rot 0 -91.3023 -89.6582>
  }]
 charm_02 [{shapeType=Capsule, length=5.34, radius=1, shapeName=charm_02_capsule, center=<3.9106 0 0>, rotation=<rot -90 0.0004 0>}]
 charm_03 [{shapeType=Capsule, length=5.1229, radius=1, shapeName=charm_03_capsule, center=<3.5082 0 0>, rotation=<rot -90 0.0004 0>}]
 charm_04 [{shapeType=Capsule, length=0.1, radius=0.825, shapeName=charm_04_capsule, center=<0.4674 0.0382 0.0009>, rotation=<rot 0 -90.3252 -89.918>}]
 charm_05 [{shapeType=Capsule, length=2.5, radius=2.9684, shapeName=charm_05_capsule, center=<3.0417 -1.2289 -1.6175>, rotation=<rot -45 0.0014 0>}]
 charm_main [{
   shapeType=Capsule
   length=1
   radius=1.3439
   shapeName=charm_main_capsule
   center=<1.2751 -0.4569 -0.011>
   rotation=<rot 9.1499 -105.1966 87.243>
  }]
 pack [{shapeType=Capsule, length=25.4333, radius=32.7954, center=<1.44 29.5423 -5.5814>, rotation=<rot -0.0245 -90 21.785>}]
}
constraints [
 {bone1Name=charm_05, bone2Name=charm_04, swing1LimitDegrees=25, swing2LimitDegrees=35, twistLimitDegrees=15, twistMotion=Locked}
 {bone1Name=charm_main, bone2Name=pack, swing1LimitDegrees=25, swing2LimitDegrees=15, twistLimitDegrees=15, twistMotion=Locked}
 {bone1Name=charm_01, bone2Name=charm_main, swing1LimitDegrees=25, swing2LimitDegrees=15, twistLimitDegrees=2}
 {bone1Name=charm_02, bone2Name=charm_01, swing1LimitDegrees=25, swing2LimitDegrees=20, twistLimitDegrees=2}
 {bone1Name=charm_03, bone2Name=charm_02, swing1LimitDegrees=20, swing2LimitDegrees=20, twistLimitDegrees=2}
 {bone1Name=charm_04, bone2Name=charm_03, swing1LimitDegrees=35, swing2LimitDegrees=25, twistLimitDegrees=15, twistMotion=Locked}
]
