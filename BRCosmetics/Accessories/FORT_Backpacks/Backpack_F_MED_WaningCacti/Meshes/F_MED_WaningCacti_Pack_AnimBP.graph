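# Anim Blueprint for the WaningCacti backpack, adding secondary motion to the pack, its flowers and leaflets.
# Event graph maps spine_05-to-spine_01 rotation delta to additive pack-bone offsets, applied before a RigidBody simulation cached as Main pose.
# Two Anim Dynamics spring chains branch from that cache, selected by bIsInFrontEnd; the front-end chain has some reduced alphas.
# It also stores simulation space and parameter-asset-driven gravity and rigid-body settings.
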AnimBlueprint /BRCosmetics/Accessories/FORT_Backpacks/Backpack_F_MED_WaningCacti/Meshes/F_MED_WaningCacti_Pack_AnimBP
target_skeleton=&/F_MED_WaningCacti_Pack_Skeleton
graphs [
 AnimGraph {nodes [
   n1 = Anim Dynamics - Bone: flower_01!(
    SimulationSpace:AnimPhysSimSpaceType Enum
    ComponentPose=Local To Component(LocalPose=Use cached pose 'Main pose'())
    Alpha=1
    GravityOverride:Vector="0,0,0"
    ExternalForce="0,0,0"
   ) -> (Pose) {
    angularDampingOverride=0.999
    angularSpringConstant=3600
    bAngularSpring=true
    bGravityOverrideInSimSpace=true
    bOverrideAngularDamping=true
    bUseGravityOverride=true
    linearDampingOverride=1
    linearSpringConstant=100
    alphaBoolBlend {blendOption=Linear}
    boundBone {boneName=flower_01}
    componentAppliedLinearAccClamp=<3000 3000 2500>
    componentLinearAccScale=<0.3 0.3 0.3>
    componentLinearVelScale=<0.3 0.3 0.3>
    physicsBodyDefinitions [{
      boundBone {boneName=flower_01}
      boxExtents=<15 15 15>
      constraintSetup {coneAngle=1, angularLimitsMax=<0 25 25>, angularLimitsMin=<0 -25 -25>, angularTarget=<1 0 0>}
      localJointOffset=<1.3 0 0>
     }]
    simSpaceSettings {maxAngularAcceleration=35, maxAngularVelocity=7, simSpaceAngularAlpha=1}
   }
   n2 = Anim Dynamics - Bone: flower_02!(SimulationSpace:AnimPhysSimSpaceType Enum, ComponentPose=@n1, Alpha=1, GravityOverride:Vector="0,0,0", ExternalForce="0,0,0") -> (Pose) {
    angularDampingOverride=0.9999
    angularSpringConstant=2800
    bAngularSpring=true
    bGravityOverrideInSimSpace=true
    bOverrideAngularDamping=true
    bUseGravityOverride=true
    linearDampingOverride=1
    linearSpringConstant=100
    alphaBoolBlend {blendOption=Linear}
    boundBone {boneName=flower_02}
    componentAppliedLinearAccClamp=<3000 3000 2500>
    componentLinearAccScale=<0.2 0.2 0.2>
    componentLinearVelScale=<0.2 0.2 0.2>
    physicsBodyDefinitions [{
      boundBone {boneName=flower_02}
      boxExtents=<15 15 15>
      constraintSetup {coneAngle=1, angularLimitsMax=<5 25 25>, angularLimitsMin=<-5 -25 -25>, angularTarget=<1 0 0>}
      localJointOffset=<1.3 0 0>
     }]
    simSpaceSettings {maxAngularAcceleration=35, maxAngularVelocity=7, simSpaceAngularAlpha=1}
   }
   n3 = Anim Dynamics - Bone: flower_03!(SimulationSpace:AnimPhysSimSpaceType Enum, ComponentPose=@n2, Alpha=1, GravityOverride:Vector="0,0,0", ExternalForce="0,0,0") -> (Pose) {
    angularDampingOverride=0.9999
    angularSpringConstant=4800
    bAngularSpring=true
    bGravityOverrideInSimSpace=true
    bOverrideAngularDamping=true
    bUseGravityOverride=true
    linearDampingOverride=1
    linearSpringConstant=100
    alphaBoolBlend {blendOption=Linear}
    boundBone {boneName=flower_03}
    componentAppliedLinearAccClamp=<3000 3000 2500>
    componentLinearAccScale=<0.35 0.35 0.35>
    componentLinearVelScale=<0.35 0.35 0.35>
    physicsBodyDefinitions [{
      boundBone {boneName=flower_03}
      boxExtents=<15 15 15>
      constraintSetup {coneAngle=1, angularLimitsMax=<0 25 25>, angularLimitsMin=<0 -25 -25>, angularTarget=<1 0 0>}
      localJointOffset=<1.3 0 0>
     }]
    simSpaceSettings {maxAngularAcceleration=35, maxAngularVelocity=7, simSpaceAngularAlpha=1}
   }
   n4 = Anim Dynamics - Bone: flower_04!(SimulationSpace:AnimPhysSimSpaceType Enum, ComponentPose=@n3, Alpha=1, GravityOverride:Vector="0,0,0", ExternalForce="0,0,0") -> (Pose) {
    angularDampingOverride=0.9999
    angularSpringConstant=1800
    bAngularSpring=true
    bGravityOverrideInSimSpace=true
    bOverrideAngularDamping=true
    bUseGravityOverride=true
    linearDampingOverride=1
    linearSpringConstant=100
    alphaBoolBlend {blendOption=Linear}
    boundBone {boneName=flower_04}
    componentAppliedLinearAccClamp=<3000 3000 2500>
    componentLinearAccScale=<0.1 0.1 0.1>
    componentLinearVelScale=<0.1 0.1 0.1>
    physicsBodyDefinitions [{
      boundBone {boneName=flower_04}
      boxExtents=<15 15 15>
      constraintSetup {coneAngle=1, angularLimitsMax=<0 25 25>, angularLimitsMin=<0 -25 -25>, angularTarget=<1 0 0>}
      localJointOffset=<1.3 0 0>
     }]
    simSpaceSettings {maxAngularAcceleration=35, maxAngularVelocity=7, simSpaceAngularAlpha=1}
   }
   n5 = Anim Dynamics - Bone: leaflet_01!(SimulationSpace:AnimPhysSimSpaceType Enum, ComponentPose=@n4, Alpha=1, GravityOverride:Vector="0,0,0", ExternalForce="0,0,0") -> (Pose) {
    angularDampingOverride=0.9999
    angularSpringConstant=1800
    bAngularSpring=true
    bGravityOverrideInSimSpace=true
    bOverrideAngularDamping=true
    bUseGravityOverride=true
    linearDampingOverride=1
    linearSpringConstant=100
    alphaBoolBlend {blendOption=Linear}
    boundBone {boneName=leaflet_01}
    componentAppliedLinearAccClamp=<3000 3000 2500>
    componentLinearAccScale=<0.25 0.25 0.25>
    componentLinearVelScale=<0.25 0.25 0.25>
    physicsBodyDefinitions [{
      boundBone {boneName=leaflet_01}
      boxExtents=<15 15 15>
      constraintSetup {coneAngle=1, angularLimitsMax=<0 25 25>, angularLimitsMin=<0 -25 -25>, angularTarget=<1 0 0>}
      localJointOffset=<1.3 0 0>
     }]
    simSpaceSettings {maxAngularAcceleration=35, maxAngularVelocity=7, simSpaceAngularAlpha=1}
   }
   n6 = Anim Dynamics - Bone: leaflet_02!(SimulationSpace:AnimPhysSimSpaceType Enum, ComponentPose=@n5, Alpha=1, GravityOverride:Vector="0,0,0", ExternalForce="0,0,0") -> (Pose) {
    angularDampingOverride=0.9996
    angularSpringConstant=6800
    bAngularSpring=true
    bGravityOverrideInSimSpace=true
    bOverrideAngularDamping=true
    bUseGravityOverride=true
    linearDampingOverride=1
    linearSpringConstant=100
    alphaBoolBlend {blendOption=Linear}
    boundBone {boneName=leaflet_02}
    componentAppliedLinearAccClamp=<3000 3000 2500>
    componentLinearAccScale=<0.1 0.1 0.1>
    componentLinearVelScale=<0.1 0.1 0.1>
    physicsBodyDefinitions [{
      boundBone {boneName=leaflet_02}
      boxExtents=<15 15 15>
      constraintSetup {coneAngle=1, angularLimitsMax=<0 25 25>, angularLimitsMin=<0 -25 -25>, angularTarget=<1 0 0>}
      localJointOffset=<1.3 0 0>
     }]
    simSpaceSettings {maxAngularAcceleration=35, maxAngularVelocity=7, simSpaceAngularAlpha=1}
   }
   n7 = Anim Dynamics - Bone: leaflet_03!(SimulationSpace:AnimPhysSimSpaceType Enum, ComponentPose=@n6, Alpha=1, GravityOverride:Vector="0,0,0", ExternalForce="0,0,0") -> (Pose) {
    angularDampingOverride=0.9999
    angularSpringConstant=5800
    bAngularSpring=true
    bGravityOverrideInSimSpace=true
    bOverrideAngularDamping=true
    bUseGravityOverride=true
    linearDampingOverride=1
    linearSpringConstant=100
    alphaBoolBlend {blendOption=Linear}
    boundBone {boneName=leaflet_03}
    componentAppliedLinearAccClamp=<3000 3000 2500>
    componentLinearAccScale=<0.1 0.1 0.1>
    componentLinearVelScale=<0.1 0.1 0.1>
    physicsBodyDefinitions [{
      boundBone {boneName=leaflet_03}
      boxExtents=<15 15 15>
      constraintSetup {coneAngle=1, angularLimitsMax=<0 25 25>, angularLimitsMin=<0 -25 -25>, angularTarget=<1 0 0>}
      localJointOffset=<1.3 0 0>
     }]
    simSpaceSettings {maxAngularAcceleration=35, maxAngularVelocity=7, simSpaceAngularAlpha=1}
   }
   n8 = Anim Dynamics - Bone: flower_01(ComponentPose=Local To Component(LocalPose=Use cached pose 'Main pose'()), Alpha=0.8, GravityOverride="0,0,0", ExternalForce="0,0,0") {
    angularDampingOverride=0.999
    angularSpringConstant=3600
    bAngularSpring=true
    bGravityOverrideInSimSpace=true
    bOverrideAngularDamping=true
    bUseGravityOverride=true
    linearDampingOverride=1
    linearSpringConstant=100
    alphaBoolBlend {blendOption=Linear}
    boundBone {boneName=flower_01}
    componentAppliedLinearAccClamp=<3000 3000 2500>
    componentLinearAccScale=<0.3 0.3 0.3>
    componentLinearVelScale=<0.3 0.3 0.3>
    physicsBodyDefinitions [{
      boundBone {boneName=flower_01}
      boxExtents=<15 15 15>
      constraintSetup {coneAngle=1, angularLimitsMax=<0 25 25>, angularLimitsMin=<0 -25 -25>, angularTarget=<1 0 0>}
      localJointOffset=<1.3 0 0>
     }]
    simSpaceSettings {maxAngularAcceleration=35, maxAngularVelocity=7, simSpaceAngularAlpha=1}
   }
   n9 = Anim Dynamics - Bone: flower_02(ComponentPose=@n8, Alpha=0.8, GravityOverride="0,0,0", ExternalForce="0,0,0") {
    angularDampingOverride=0.9999
    angularSpringConstant=2800
    bAngularSpring=true
    bGravityOverrideInSimSpace=true
    bOverrideAngularDamping=true
    bUseGravityOverride=true
    linearDampingOverride=1
    linearSpringConstant=100
    alphaBoolBlend {blendOption=Linear}
    boundBone {boneName=flower_02}
    componentAppliedLinearAccClamp=<3000 3000 2500>
    componentLinearAccScale=<0.2 0.2 0.2>
    componentLinearVelScale=<0.2 0.2 0.2>
    physicsBodyDefinitions [{
      boundBone {boneName=flower_02}
      boxExtents=<15 15 15>
      constraintSetup {coneAngle=1, angularLimitsMax=<5 25 25>, angularLimitsMin=<-5 -25 -25>, angularTarget=<1 0 0>}
      localJointOffset=<1.3 0 0>
     }]
    simSpaceSettings {maxAngularAcceleration=35, maxAngularVelocity=7, simSpaceAngularAlpha=1}
   }
   n10 = Anim Dynamics - Bone: flower_03(ComponentPose=@n9, Alpha=1, GravityOverride="0,0,0", ExternalForce="0,0,0") {
    angularDampingOverride=0.9999
    angularSpringConstant=4800
    bAngularSpring=true
    bGravityOverrideInSimSpace=true
    bOverrideAngularDamping=true
    bUseGravityOverride=true
    linearDampingOverride=1
    linearSpringConstant=100
    alphaBoolBlend {blendOption=Linear}
    boundBone {boneName=flower_03}
    componentAppliedLinearAccClamp=<3000 3000 2500>
    componentLinearAccScale=<0.35 0.35 0.35>
    componentLinearVelScale=<0.35 0.35 0.35>
    physicsBodyDefinitions [{
      boundBone {boneName=flower_03}
      boxExtents=<15 15 15>
      constraintSetup {coneAngle=1, angularLimitsMax=<0 25 25>, angularLimitsMin=<0 -25 -25>, angularTarget=<1 0 0>}
      localJointOffset=<1.3 0 0>
     }]
    simSpaceSettings {maxAngularAcceleration=35, maxAngularVelocity=7, simSpaceAngularAlpha=1}
   }
   n11 = Anim Dynamics - Bone: flower_04(ComponentPose=@n10, Alpha=1, GravityOverride="0,0,0", ExternalForce="0,0,0") {
    angularDampingOverride=0.9999
    angularSpringConstant=1800
    bAngularSpring=true
    bGravityOverrideInSimSpace=true
    bOverrideAngularDamping=true
    bUseGravityOverride=true
    linearDampingOverride=1
    linearSpringConstant=100
    alphaBoolBlend {blendOption=Linear}
    boundBone {boneName=flower_04}
    componentAppliedLinearAccClamp=<3000 3000 2500>
    componentLinearAccScale=<0.1 0.1 0.1>
    componentLinearVelScale=<0.1 0.1 0.1>
    physicsBodyDefinitions [{
      boundBone {boneName=flower_04}
      boxExtents=<15 15 15>
      constraintSetup {coneAngle=1, angularLimitsMax=<0 25 25>, angularLimitsMin=<0 -25 -25>, angularTarget=<1 0 0>}
      localJointOffset=<1.3 0 0>
     }]
    simSpaceSettings {maxAngularAcceleration=35, maxAngularVelocity=7, simSpaceAngularAlpha=1}
   }
   n12 = Anim Dynamics - Bone: leaflet_01(ComponentPose=@n11, Alpha=1, GravityOverride="0,0,0", ExternalForce="0,0,0") {
    angularDampingOverride=0.9999
    angularSpringConstant=1800
    bAngularSpring=true
    bGravityOverrideInSimSpace=true
    bOverrideAngularDamping=true
    bUseGravityOverride=true
    linearDampingOverride=1
    linearSpringConstant=100
    alphaBoolBlend {blendOption=Linear}
    boundBone {boneName=leaflet_01}
    componentAppliedLinearAccClamp=<3000 3000 2500>
    componentLinearAccScale=<0.25 0.25 0.25>
    componentLinearVelScale=<0.25 0.25 0.25>
    physicsBodyDefinitions [{
      boundBone {boneName=leaflet_01}
      boxExtents=<15 15 15>
      constraintSetup {coneAngle=1, angularLimitsMax=<0 25 25>, angularLimitsMin=<0 -25 -25>, angularTarget=<1 0 0>}
      localJointOffset=<1.3 0 0>
     }]
    simSpaceSettings {maxAngularAcceleration=35, maxAngularVelocity=7, simSpaceAngularAlpha=1}
   }
   n13 = Anim Dynamics - Bone: leaflet_02(ComponentPose=@n12, Alpha=0.7, GravityOverride="0,0,0", ExternalForce="0,0,0") {
    angularDampingOverride=0.9996
    angularSpringConstant=6800
    bAngularSpring=true
    bGravityOverrideInSimSpace=true
    bOverrideAngularDamping=true
    bUseGravityOverride=true
    linearDampingOverride=1
    linearSpringConstant=100
    alphaBoolBlend {blendOption=Linear}
    boundBone {boneName=leaflet_02}
    componentAppliedLinearAccClamp=<3000 3000 2500>
    componentLinearAccScale=<0.1 0.1 0.1>
    componentLinearVelScale=<0.1 0.1 0.1>
    physicsBodyDefinitions [{
      boundBone {boneName=leaflet_02}
      boxExtents=<15 15 15>
      constraintSetup {coneAngle=1, angularLimitsMax=<0 25 25>, angularLimitsMin=<0 -25 -25>, angularTarget=<1 0 0>}
      localJointOffset=<1.3 0 0>
     }]
    simSpaceSettings {maxAngularAcceleration=35, maxAngularVelocity=7, simSpaceAngularAlpha=1}
   }
   n14 = Anim Dynamics - Bone: leaflet_03(ComponentPose=@n13, Alpha=0.5, GravityOverride="0,0,0", ExternalForce="0,0,0") {
    angularDampingOverride=0.9999
    angularSpringConstant=5800
    bAngularSpring=true
    bGravityOverrideInSimSpace=true
    bOverrideAngularDamping=true
    bUseGravityOverride=true
    linearDampingOverride=1
    linearSpringConstant=100
    alphaBoolBlend {blendOption=Linear}
    boundBone {boneName=leaflet_03}
    componentAppliedLinearAccClamp=<3000 3000 2500>
    componentLinearAccScale=<0.1 0.1 0.1>
    componentLinearVelScale=<0.1 0.1 0.1>
    physicsBodyDefinitions [{
      boundBone {boneName=leaflet_03}
      boxExtents=<15 15 15>
      constraintSetup {coneAngle=1, angularLimitsMax=<0 25 25>, angularLimitsMin=<0 -25 -25>, angularTarget=<1 0 0>}
      localJointOffset=<1.3 0 0>
     }]
    simSpaceSettings {maxAngularAcceleration=35, maxAngularVelocity=7, simSpaceAngularAlpha=1}
   }
   Output Pose(Result=Blend Poses by bool(BlendPose_0=Component To Local(ComponentPose=@n14), BlendPose_1=Component To Local(ComponentPose=@n7), bActiveValue=Get bIsInFrontEnd())) {layerGroup=None}
   n15 = "Transform (Modify) Bone - Bone: pack"(ComponentPose=Local To Component(LocalPose=Input Pose()), Translation=Get backpack_translation(), Rotation=Get backpack_rotate()) {rotationMode=BMM_Additive, translationMode=BMM_Additive, alphaBoolBlend {blendOption=Linear}, boneToModify {boneName=pack}}
   n16 = "Transform (Modify) Bone - Bone: pack"(
    ComponentPose=@n15
    Translation=Get Backpack_translation_X() -> (Backpack_translation_X:Vector)
    Rotation=Get Backpack_rotate_X() -> (Backpack_rotate_X:Rotator)
   ) {rotationMode=BMM_Additive, translationMode=BMM_Additive, alphaBoolBlend {blendOption=Linear}, boneToModify {boneName=pack}}
   n17 = "Transform (Modify) Bone - Bone: pack"(ComponentPose=@n16, Translation="0,0,0", Rotation=Get Backpack_rotate_-X() -> (Backpack_rotate_-X:Rotator)) {rotationMode=BMM_Additive, translationMode=BMM_Additive, alphaBoolBlend {blendOption=Linear}, boneToModify {boneName=pack}}
   n18 = RigidBody!(
    ComponentPose=@n17
    Alpha=0.74
    OverrideWorldGravity:Vector="0,0,0"
    ExternalForce
    SimSpaceSettings:Sim Space Settings Structure=(WorldAlpha=1,VelocityScaleZ=0.5,DampingAlpha=1,MaxLinearVelocity=800,MaxAngularVelocity=8,MaxLinearAcceleration=2400,MaxAngularAcceleration=32,ExternalLinearDragV=(X=0.25,Y=0.25,Z=0.25),ExternalLinearVelocity=(X=0,Y=0,Z=0),ExternalAngularVelocity=(X=0,Y=0,Z=0))
   ) -> (Pose) {
    bClampLinearTranslationLimitToRefPose=true
    bOverrideWorldGravity=true
    overridePhysicsAsset=&/F_MED_WaningCacti_Physics
    simulationSpace=BaseBoneSpace
    alphaBoolBlend {blendOption=Linear}
    baseBoneRef {boneName=pack}
    componentAppliedLinearAccClamp=<3000 3000 3000>
    componentLinearAccScale=<0.2 0.2 0.2>
    componentLinearVelScale=<0.2 0.2 0.3>
    overrideWorldGravity {z=-850}
    simSpaceSettings {
     maxAngularAcceleration=32
     maxAngularVelocity=8
     maxLinearAcceleration=2400
     maxLinearVelocity=800
     velocityScaleZ=0.5
     worldAlpha=1
     externalLinearDragV=<0.25 0.25 0.25>
    }
   }
   Save cached pose 'Main pose'(Pose=Component To Local(ComponentPose=@n18)) {cacheName=Main pose}
  ]}
 EventGraph {nodes [
   Event BlueprintUpdateAnimation()
   > n1 = Sequence!(execute) -> (then_0, then_1, then_2:Exec, then_3:Exec, then_4:Exec)
   n2 = GetOwnerSkeletalMeshComponent()
   n3 = BreakRotator(InRot=BreakTransform(InTransform=GetDeltaTransformFromRefPose(self=@n2, BoneName=spine_05, BaseName=spine_01)).Rotation)
   n4 = MapRangeClamped(Value=@n3.Yaw, InRangeA=0, InRangeB=-20, OutRangeB=-6)
   n5 = MapRangeClamped(Value=@n3.Yaw, InRangeA=-20, InRangeB=-50, OutRangeB=-9)
   > Set backpack_translation(backpack_translation=MakeVector(
     X=MapRangeClamped(Value=@n3.Yaw, InRangeA=-36, InRangeB=-60, OutRangeB=-5)
     Y=float + float(A=@n4, B=@n5)
     Z=MapRangeClamped(Value=@n3.Yaw, InRangeA=-36, InRangeB=-60, OutRangeB=4)
    ))
   n6 = MapRangeClamped(Value=@n3.Yaw, InRangeA=-30, InRangeB=-150, OutRangeB=130)
   n7 = MapRangeClamped(Value=@n3.Yaw, InRangeA=0, InRangeB=30, OutRangeB=-15)
   n8 = MapRangeClamped(Value=@n3.Yaw, InRangeA=-4, InRangeB=-30, OutRangeB=25)
   n9 = float + float(
    A=MapRangeClamped(Value=@n3.Yaw, InRangeA=-17, InRangeB=-40, OutRangeB=-16)
    B=MapRangeClamped(Value=@n3.Yaw, InRangeA=-36, InRangeB=-150, OutRangeB=-100)
   )
   > Set backpack_rotate(backpack_rotate=MakeRotator(Roll=float + float!(A=@n8, B=@n6, C:double=@n7) -> (ReturnValue), Pitch=@n9, Yaw=@n9))
   > Set Backpack_rotate_X(Backpack_rotate_X:Rotator=MakeRotator(
     Pitch=MapRangeClamped(Value=@n3, InRangeA=10, InRangeB=45, OutRangeB=10)
     Yaw=MapRangeClamped(Value=@n3, InRangeA=10, InRangeB=45, OutRangeB=-20)
    )) -> (Output_Get:Rotator)
   > Set Backpack_translation_X(Backpack_translation_X:Vector=MakeVector(X=MapRangeClamped(Value=@n3, InRangeA=10, InRangeB=45, OutRangeB=-5), Y=MapRangeClamped(Value=@n3, InRangeA=10, InRangeB=45, OutRangeB=-4))) -> (Output_Get:Vector)
   > Set Backpack_rotate_-X(Backpack_rotate_-X:Rotator=MakeRotator(
     Roll=MapRangeClamped(Value=@n3, InRangeA=-10, InRangeB=-45, OutRangeB=15)
     Pitch=MapRangeClamped(Value=@n3, InRangeA=-10, InRangeB=-45, OutRangeB=-20)
    )) -> (Output_Get:Rotator)
   > n10 = Branch(Condition=Get bIsInFrontEnd())
   > Set SimlationSpace(SimlationSpace:AnimPhysSimSpaceType Enum=Component) -> (Output_Get:AnimPhysSimSpaceType Enum)
   Set SimlationSpace(execute=@n10.else)
   n11 = "Gravity Override State Function (Parameter Asset) (Nativized)"(
    execute=@n1.then_1
    CharacterPartAnimInstance=Self-Reference()
    GameTimeInSeconds=GetGameTimeInSeconds()
    Parameters=&/F_MED_WaningCacti_Parameters
   )
   > Set Gravity Override(Gravity Override=@n11)
   > Set RBAN Settings Out(RBAN Settings Out=@n11.OutRigidBodyAnimNodeParameters)
   TryGetPawnOwner()
  ]}
]
variables [
 backpack_rotate {type=Rotator}
 Backpack_rotate_-X {type=Rotator}
 Backpack_rotate_X {type=Rotator}
 backpack_translation {type=Vector}
 Backpack_translation_X {type=Vector}
 Gravity Override {type=Vector}
 RBAN Settings Out {type=FortRigidBodyAnimNodeParameters}
 SimlationSpace {type=AnimPhysSimSpaceType}
]
